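# Physics asset for the MimicGlean medium male player body, covering a dangling bag and the lower-body area around it, rather than a full ragdoll.
# Capsules on pelvis, spine_01 and left thigh form the body volumes near the two-bone dyn_bag chain.
# The bag base joins the pelvis with free rotation, while the bag has wide swing and tight twist limits against its base.
# Spine_01 also rotates freely on the pelvis.
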PhysicsAsset /BRCosmetics/Characters/Player/Male/Medium/Bodies/M_MED_MimicGlean/Meshes/M_MED_MimicGlean_Physics
bodies {
 dyn_bag [
  {shapeType=Capsule, length=14, radius=5, shapeName=dyn_bag_capsule, center=<5 0 0>, rotation=<rot 0 0 90>}
  {shapeType=Capsule, length=14, radius=5, shapeName=dyn_bag_capsule, center=<5 0 -15>, rotation=<rot 0 0 90>}
 ]
 dyn_bag_base [{shapeType=Capsule, length=1, radius=3, shapeName=dyn_bag_base_capsule}]
 pelvis [{shapeType=Capsule, length=12, radius=12}, {shapeType=Capsule, length=12, radius=12, shapeName=pelvis_capsule, center=<-8.9772 0 0>}]
 spine_01 [{shapeType=Capsule, length=0.505, radius=0.505, shapeName=spine_01_capsule, rotation=<rot 0 -90 0>}]
 thigh_l [{shapeType=Capsule, length=19.9912, radius=12, center=<-13.456 0.6062 3.2171>, rotation=<rot 0.0005 90.543 -89.997>}]
}
constraints [
 {bone1Name=thigh_l, bone2Name=pelvis, swing1LimitDegrees=45, swing2LimitDegrees=45, twistLimitDegrees=45}
 {
  bone1Name=spine_01
  bone2Name=pelvis
  swing1LimitDegrees=45
  swing1Motion=Free
  swing2LimitDegrees=45
  swing2Motion=Free
  twistLimitDegrees=45
  twistMotion=Free
 }
 {bone1Name=dyn_bag, bone2Name=dyn_bag_base, swing1LimitDegrees=60, swing2LimitDegrees=60, twistLimitDegrees=10}
 {
  bone1Name=dyn_bag_base
  bone2Name=pelvis
  swing1LimitDegrees=45
  swing1Motion=Free
  swing2LimitDegrees=45
  swing2Motion=Free
  twistLimitDegrees=45
  twistMotion=Free
 }
]
