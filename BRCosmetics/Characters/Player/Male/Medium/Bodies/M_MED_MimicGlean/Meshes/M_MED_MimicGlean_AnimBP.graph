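# Animation Blueprint for the MimicGlean male medium body cosmetic, adding secondary motion to the pose copied from the attached parent mesh.
# A pre-dynamics Control Rig, bag-bone constraint and pelvis-space RigidBody simulation form one branch of a bool blend against the plain copied pose.
# Each update computes GravityOverride using parameters borrowed from the female GoldenValley body.
# BindPose starts true and clears once character customization loads.
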AnimBlueprint /BRCosmetics/Characters/Player/Male/Medium/Bodies/M_MED_MimicGlean/Meshes/M_MED_MimicGlean_AnimBP
target_skeleton=&/M_MED_MimicGlean_Skeleton
graphs [
 AnimGraph {nodes [
   n1 = Control Rig(Source=Copy Pose From Mesh() {bUseAttachedParent=true}) {controlRigAssetReference {controlRigAsset=&/M_MED_MimicGlean_CtrlRig_PreDyn}}
   n2 = "Constraint - dyn_bag_base (P)"(ComponentPose=Local To Component(LocalPose=@n1), Alpha=1, ConstraintWeights_0=1) {boneToModify {boneName=dyn_bag_base}, constraintSetup [{targetBone {boneName=extra_belt_l_side}}], constraintWeights [1]}
   n3 = RigidBody!(
    ComponentPose=@n2
    bAlphaBoolEnabled:bool
    OverrideWorldGravity:Vector="0,0,0"
    ComponentLinearAccScale:Vector="0,0,0"
    ComponentLinearVelScale:Vector="0,0,0"
    bTransferBoneVelocities:bool=False
   ) -> (Pose) {
    alphaInputType=Bool
    bClampLinearTranslationLimitToRefPose=true
    bFreezeIncomingPoseOnStart=true
    bOverrideWorldGravity=true
    bTransferBoneVelocities=true
    overridePhysicsAsset=&/M_MED_MimicGlean_Physics
    simulationSpace=BaseBoneSpace
    alphaBoolBlend {blendOption=Linear}
    baseBoneRef {boneName=pelvis}
    componentAppliedLinearAccClamp=<100 100 100>
    componentLinearAccScale=<0.6 0.2 0.2>
    componentLinearVelScale=<0.6 0.2 0.2>
    simSpaceSettings {
     dampingAlpha=0
     maxAngularAcceleration=50
     maxAngularVelocity=8
     maxLinearAcceleration=1000
     maxLinearVelocity=500
     worldAlpha=1
     externalLinearDragV=<1 1 1>
    }
   }
   Output Pose(Result=Blend Poses by bool(BlendPose_0=Copy Pose From Mesh() {bUseAttachedParent=true}, BlendPose_1=Component To Local(ComponentPose=@n3))) {layerGroup=None}
  ]}
 EventGraph {nodes [
   Event BlueprintUpdateAnimation()
   > Sequence()
   > n1 = "Gravity Override State Function (Parameter Asset) (Nativized)"(
    CharacterPartAnimInstance=Self-Reference()
    GameTimeInSeconds=GetGameTimeInSeconds()
    DebugJointName=spine_05
    Parameters=/BRCosmetics/Characters/Player/Female/Medium/Bodies/F_MED_GoldenValley/Meshes/F_MED_GoldenValley_Parameters
   )
   > Set GravityOverride(GravityOverride=@n1)
   Event OnCharacterCustomizationLoaded()
   > Set BindPose(BindPose:bool=false) -> (Output_Get:bool)
  ]}
]
variables [
 BindPose {type=boolean, default=true}
 GravityOverride {type=Vector}
]
