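# Physics asset for the MimicGlean male medium face accessory mesh, defining bodies for its dynamic hair bones plus head and neck.
# Each hair lock pairs a small base capsule, angularly locked to the head, with an offset capsule constrained to that base by swing and twist limits.
# Back locks have tighter swing limits than front, side and top ones.
# Larger head and neck capsules form a chain with free angular motion.
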PhysicsAsset /BRCosmetics/Characters/Player/Male/Medium/Bodies/M_MED_MimicGlean/Meshes/Parts/M_MED_MimicGlean_FaceAcc_Physics
bodies {
 dyn_hair_bk_01 [{shapeType=Capsule, length=2, radius=2, shapeName=dyn_hair_bk_01_capsule, center=<5 0 0>, rotation=<rot 0 0 90>}]
 dyn_hair_bk_01_base [{shapeType=Capsule, length=1, radius=1, shapeName=dyn_hair_bk_01_base_capsule}]
 dyn_hair_bk_02 [{shapeType=Capsule, length=2, radius=2, shapeName=dyn_hair_bk_02_capsule, center=<5 0 0>, rotation=<rot 0 0 90>}]
 dyn_hair_bk_02_base [{shapeType=Capsule, length=1, radius=1, shapeName=dyn_hair_bk_02_base_capsule}]
 dyn_hair_bk_03 [{shapeType=Capsule, length=2, radius=2, shapeName=dyn_hair_bk_03_capsule, center=<5 0 0>, rotation=<rot 0 0 90>}]
 dyn_hair_bk_03_base [{shapeType=Capsule, length=1, radius=1, shapeName=dyn_hair_bk_03_base_capsule}]
 dyn_hair_bk_04 [{shapeType=Capsule, length=2, radius=2, shapeName=dyn_hair_bk_04_capsule, center=<5 0 0>, rotation=<rot 0 0 90>}]
 dyn_hair_bk_04_base [{shapeType=Capsule, length=1, radius=1, shapeName=dyn_hair_bk_04_base_capsule}]
 dyn_hair_fr_01 [{shapeType=Capsule, length=2, radius=2, shapeName=dyn_hair_fr_01_capsule, center=<5 0 0>}]
 dyn_hair_fr_01_base [{shapeType=Capsule, length=1, radius=1, shapeName=dyn_hair_fr_01_base_capsule}]
 dyn_hair_fr_side_l_01 [{shapeType=Capsule, length=2, radius=2, shapeName=dyn_hair_fr_side_l_01_capsule, center=<5 0 0>, rotation=<rot 0 0 90>}]
 dyn_hair_fr_side_l_01_base [{shapeType=Capsule, length=1, radius=1, shapeName=dyn_hair_fr_side_l_01_base_capsule}]
 dyn_hair_fr_side_r_01 [{shapeType=Capsule, length=2, radius=2, shapeName=dyn_hair_fr_side_r_01_capsule, center=<5 0 0>, rotation=<rot 0 0 90>}]
 dyn_hair_fr_side_r_01_base [{shapeType=Capsule, length=1, radius=1, shapeName=dyn_hair_fr_side_r_01_base_capsule}]
 dyn_hair_side_l_01 [{shapeType=Capsule, length=2, radius=2, shapeName=dyn_hair_side_l_01_capsule, center=<5 0 0>, rotation=<rot 90 0 0>}]
 dyn_hair_side_l_01_base [{shapeType=Capsule, length=1, radius=1, shapeName=dyn_hair_side_l_01_base_capsule}]
 dyn_hair_side_r_01 [{shapeType=Capsule, length=2, radius=2, shapeName=dyn_hair_side_r_01_capsule, center=<5 0 0>, rotation=<rot 90 0 0>}]
 dyn_hair_side_r_01_base [{shapeType=Capsule, length=1, radius=1, shapeName=dyn_hair_side_r_01_base_capsule}]
 dyn_hair_top_01 [{shapeType=Capsule, length=2, radius=2, shapeName=dyn_hair_top_01_capsule, center=<5 0 0>, rotation=<rot 0 0 90>}]
 dyn_hair_top_01_base [{shapeType=Capsule, length=1, radius=1, shapeName=dyn_hair_top_01_base_capsule}]
 dyn_hair_top_02 [{shapeType=Capsule, length=2, radius=2, shapeName=dyn_hair_top_02_capsule, center=<5 0 0>, rotation=<rot 0 0 90>}]
 dyn_hair_top_02_base [{shapeType=Capsule, length=1, radius=1, shapeName=dyn_hair_top_02_base_capsule}]
 dyn_hair_top_03 [{shapeType=Capsule, length=2, radius=2, shapeName=dyn_hair_top_03_capsule, center=<5 0 0>, rotation=<rot 0 0 90>}]
 dyn_hair_top_03_base [{shapeType=Capsule, length=1, radius=1, shapeName=dyn_hair_top_03_base_capsule}]
 dyn_hair_top_04 [{shapeType=Capsule, length=2, radius=2, shapeName=dyn_hair_top_04_capsule, center=<5 0 0>, rotation=<rot 0 0 90>}]
 dyn_hair_top_04_base [{shapeType=Capsule, length=1, radius=1, shapeName=dyn_hair_top_04_base_capsule}]
 dyn_hair_top_05 [{shapeType=Capsule, length=2, radius=2, shapeName=dyn_hair_top_05_capsule, center=<5 0 0>, rotation=<rot 0 0 90>}]
 dyn_hair_top_05_base [{shapeType=Capsule, length=1, radius=1, shapeName=dyn_hair_top_05_base_capsule}]
 dyn_hair_top_06 [{shapeType=Capsule, length=2, radius=2, shapeName=dyn_hair_top_06_capsule, center=<5 0 0>, rotation=<rot 0 0 90>}]
 dyn_hair_top_06_base [{shapeType=Capsule, length=1, radius=1, shapeName=dyn_hair_top_06_base_capsule}]
 dyn_hair_top_07 [{shapeType=Capsule, length=2, radius=2, shapeName=dyn_hair_top_07_capsule, center=<5 0 0>, rotation=<rot 0 0 90>}]
 dyn_hair_top_07_base [{shapeType=Capsule, length=1, radius=1, shapeName=dyn_hair_top_07_base_capsule}]
 dyn_hair_top_08 [{shapeType=Capsule, length=2, radius=2, shapeName=dyn_hair_top_08_capsule, center=<5 0 0>, rotation=<rot 0 0 90>}]
 dyn_hair_top_08_base [{shapeType=Capsule, length=1, radius=1, shapeName=dyn_hair_top_08_base_capsule}]
 dyn_hair_top_fr_01 [{shapeType=Capsule, length=2, radius=2, shapeName=dyn_hair_top_fr_01_capsule, center=<5 0 0>, rotation=<rot 0 0 90>}]
 dyn_hair_top_fr_01_base [{shapeType=Capsule, length=1, radius=1, shapeName=dyn_hair_top_fr_01_base_capsule}]
 head [{shapeType=Capsule, length=10, radius=10, center=<5.5937 -0.0001 0>, rotation=<rot -90 -0.0006 0>}]
 neck_01 [{shapeType=Capsule, length=5, radius=8, shapeName=neck_01_capsule, rotation=<rot 90 0 0>}]
 neck_02 [{shapeType=Capsule, length=5, radius=8, shapeName=neck_02_capsule, rotation=<rot 90 0 0>}]
}
constraints [
 {
  bone1Name=dyn_hair_bk_01_base
  bone2Name=head
  swing1LimitDegrees=45
  swing1Motion=Locked
  swing2LimitDegrees=45
  swing2Motion=Locked
  twistLimitDegrees=45
  twistMotion=Locked
 }
 {bone1Name=dyn_hair_bk_01, bone2Name=dyn_hair_bk_01_base, swing1LimitDegrees=15, swing2LimitDegrees=30, twistLimitDegrees=5}
 {
  bone1Name=dyn_hair_bk_02_base
  bone2Name=head
  swing1LimitDegrees=45
  swing1Motion=Locked
  swing2LimitDegrees=45
  swing2Motion=Locked
  twistLimitDegrees=45
  twistMotion=Locked
 }
 {bone1Name=dyn_hair_bk_02, bone2Name=dyn_hair_bk_02_base, swing1LimitDegrees=15, swing2LimitDegrees=30, twistLimitDegrees=5}
 {
  bone1Name=dyn_hair_bk_03_base
  bone2Name=head
  swing1LimitDegrees=45
  swing1Motion=Locked
  swing2LimitDegrees=45
  swing2Motion=Locked
  twistLimitDegrees=45
  twistMotion=Locked
 }
 {bone1Name=dyn_hair_bk_03, bone2Name=dyn_hair_bk_03_base, swing1LimitDegrees=15, swing2LimitDegrees=30, twistLimitDegrees=5}
 {
  bone1Name=dyn_hair_bk_04_base
  bone2Name=head
  swing1LimitDegrees=45
  swing1Motion=Locked
  swing2LimitDegrees=45
  swing2Motion=Locked
  twistLimitDegrees=45
  twistMotion=Locked
 }
 {bone1Name=dyn_hair_bk_04, bone2Name=dyn_hair_bk_04_base, swing1LimitDegrees=15, swing2LimitDegrees=30, twistLimitDegrees=5}
 {
  bone1Name=dyn_hair_fr_01_base
  bone2Name=head
  swing1LimitDegrees=45
  swing1Motion=Locked
  swing2LimitDegrees=45
  swing2Motion=Locked
  twistLimitDegrees=45
  twistMotion=Locked
 }
 {bone1Name=dyn_hair_fr_01, bone2Name=dyn_hair_fr_01_base, swing1LimitDegrees=30, swing2LimitDegrees=60, twistLimitDegrees=5}
 {
  bone1Name=dyn_hair_fr_side_l_01_base
  bone2Name=head
  swing1LimitDegrees=45
  swing1Motion=Locked
  swing2LimitDegrees=45
  swing2Motion=Locked
  twistLimitDegrees=45
  twistMotion=Locked
 }
 {bone1Name=dyn_hair_fr_side_l_01, bone2Name=dyn_hair_fr_side_l_01_base, swing1LimitDegrees=30, swing2LimitDegrees=60, twistLimitDegrees=5}
 {
  bone1Name=dyn_hair_fr_side_r_01_base
  bone2Name=head
  swing1LimitDegrees=45
  swing1Motion=Locked
  swing2LimitDegrees=45
  swing2Motion=Locked
  twistLimitDegrees=45
  twistMotion=Locked
 }
 {bone1Name=dyn_hair_fr_side_r_01, bone2Name=dyn_hair_fr_side_r_01_base, swing1LimitDegrees=30, swing2LimitDegrees=60, twistLimitDegrees=5}
 {
  bone1Name=dyn_hair_side_l_01_base
  bone2Name=head
  swing1LimitDegrees=45
  swing1Motion=Locked
  swing2LimitDegrees=45
  swing2Motion=Locked
  twistLimitDegrees=45
  twistMotion=Locked
 }
 {bone1Name=dyn_hair_side_l_01, bone2Name=dyn_hair_side_l_01_base, swing1LimitDegrees=30, swing2LimitDegrees=60, twistLimitDegrees=5}
 {
  bone1Name=dyn_hair_side_r_01_base
  bone2Name=head
  swing1LimitDegrees=45
  swing1Motion=Locked
  swing2LimitDegrees=45
  swing2Motion=Locked
  twistLimitDegrees=45
  twistMotion=Locked
 }
 {bone1Name=dyn_hair_side_r_01, bone2Name=dyn_hair_side_r_01_base, swing1LimitDegrees=30, swing2LimitDegrees=60, twistLimitDegrees=5}
 {
  bone1Name=dyn_hair_top_01_base
  bone2Name=head
  swing1LimitDegrees=45
  swing1Motion=Locked
  swing2LimitDegrees=45
  swing2Motion=Locked
  twistLimitDegrees=45
  twistMotion=Locked
 }
 {bone1Name=dyn_hair_top_01, bone2Name=dyn_hair_top_01_base, swing1LimitDegrees=30, swing2LimitDegrees=60, twistLimitDegrees=5}
 {
  bone1Name=dyn_hair_top_02_base
  bone2Name=head
  swing1LimitDegrees=45
  swing1Motion=Locked
  swing2LimitDegrees=45
  swing2Motion=Locked
  twistLimitDegrees=45
  twistMotion=Locked
 }
 {bone1Name=dyn_hair_top_02, bone2Name=dyn_hair_top_02_base, swing1LimitDegrees=30, swing2LimitDegrees=60, twistLimitDegrees=5}
 {
  bone1Name=dyn_hair_top_03_base
  bone2Name=head
  swing1LimitDegrees=45
  swing1Motion=Locked
  swing2LimitDegrees=45
  swing2Motion=Locked
  twistLimitDegrees=45
  twistMotion=Locked
 }
 {bone1Name=dyn_hair_top_03, bone2Name=dyn_hair_top_03_base, swing1LimitDegrees=30, swing2LimitDegrees=60, twistLimitDegrees=5}
 {
  bone1Name=dyn_hair_top_04_base
  bone2Name=head
  swing1LimitDegrees=45
  swing1Motion=Locked
  swing2LimitDegrees=45
  swing2Motion=Locked
  twistLimitDegrees=45
  twistMotion=Locked
 }
 {bone1Name=dyn_hair_top_04, bone2Name=dyn_hair_top_04_base, swing1LimitDegrees=30, swing2LimitDegrees=60, twistLimitDegrees=5}
 {
  bone1Name=dyn_hair_top_05_base
  bone2Name=head
  swing1LimitDegrees=45
  swing1Motion=Locked
  swing2LimitDegrees=45
  swing2Motion=Locked
  twistLimitDegrees=45
  twistMotion=Locked
 }
 {bone1Name=dyn_hair_top_05, bone2Name=dyn_hair_top_05_base, swing1LimitDegrees=30, swing2LimitDegrees=60, twistLimitDegrees=5}
 {
  bone1Name=dyn_hair_top_06_base
  bone2Name=head
  swing1LimitDegrees=45
  swing1Motion=Locked
  swing2LimitDegrees=45
  swing2Motion=Locked
  twistLimitDegrees=45
  twistMotion=Locked
 }
 {bone1Name=dyn_hair_top_06, bone2Name=dyn_hair_top_06_base, swing1LimitDegrees=30, swing2LimitDegrees=60, twistLimitDegrees=5}
 {
  bone1Name=dyn_hair_top_07_base
  bone2Name=head
  swing1LimitDegrees=45
  swing1Motion=Locked
  swing2LimitDegrees=45
  swing2Motion=Locked
  twistLimitDegrees=45
  twistMotion=Locked
 }
 {bone1Name=dyn_hair_top_07, bone2Name=dyn_hair_top_07_base, swing1LimitDegrees=30, swing2LimitDegrees=60, twistLimitDegrees=5}
 {
  bone1Name=dyn_hair_top_08_base
  bone2Name=head
  swing1LimitDegrees=45
  swing1Motion=Locked
  swing2LimitDegrees=45
  swing2Motion=Locked
  twistLimitDegrees=45
  twistMotion=Locked
 }
 {bone1Name=dyn_hair_top_08, bone2Name=dyn_hair_top_08_base, swing1LimitDegrees=30, swing2LimitDegrees=60, twistLimitDegrees=5}
 {
  bone1Name=dyn_hair_top_fr_01_base
  bone2Name=head
  swing1LimitDegrees=45
  swing1Motion=Locked
  swing2LimitDegrees=45
  swing2Motion=Locked
  twistLimitDegrees=45
  twistMotion=Locked
 }
 {bone1Name=dyn_hair_top_fr_01, bone2Name=dyn_hair_top_fr_01_base, swing1LimitDegrees=30, swing2LimitDegrees=60, twistLimitDegrees=5}
 {
  bone1Name=head
  bone2Name=neck_02
  swing1LimitDegrees=45
  swing1Motion=Free
  swing2LimitDegrees=45
  swing2Motion=Free
  twistLimitDegrees=45
  twistMotion=Free
 }
 {
  bone1Name=neck_02
  bone2Name=neck_01
  swing1LimitDegrees=45
  swing1Motion=Free
  swing2LimitDegrees=45
  swing2Motion=Free
  twistLimitDegrees=45
  twistMotion=Free
 }
]
