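# Post-process style AnimBlueprint for the MimicGlean male face accessory part, adding secondary motion to its dynamic hair.
# It copies the attached parent's pose, runs a PreDyn Control Rig, constrains four back-hair base bones between neck and head, then simulates RigidBody physics in head space.
# The update event computes GravityOverride using the female GoldenValley parameter asset, but the AnimGraph never reads it.
# A second RigidBody node is left unconnected.
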AnimBlueprint /BRCosmetics/Characters/Player/Male/Medium/Bodies/M_MED_MimicGlean/Meshes/Parts/M_MED_MimicGlean_FaceAcc_AnimBP
target_skeleton=&/M_MED_MimicGlean_FaceAcc_Skeleton
graphs [
 AnimGraph {nodes [
   RigidBody!(
    ComponentPose
    bAlphaBoolEnabled:bool
    OverrideWorldGravity:Vector="0,0,0"
    ComponentLinearAccScale:Vector="0,0,0"
    ComponentLinearVelScale:Vector="0,0,0"
    bTransferBoneVelocities:bool=False
   ) -> (Pose) {
    alphaInputType=Bool
    bClampLinearTranslationLimitToRefPose=true
    bFreezeIncomingPoseOnStart=true
    bOverrideWorldGravity=true
    bTransferBoneVelocities=true
    overridePhysicsAsset=&/M_MED_MimicGlean_FaceAcc_Physics
    simulationSpace=BaseBoneSpace
    alphaBoolBlend {blendOption=Linear}
    baseBoneRef {boneName=head}
    componentAppliedLinearAccClamp=<100 100 100>
    componentLinearAccScale=<0.6 0.2 0.2>
    componentLinearVelScale=<0.6 0.2 0.2>
    simSpaceSettings {
     dampingAlpha=0
     maxAngularAcceleration=50
     maxAngularVelocity=8
     maxLinearAcceleration=1000
     maxLinearVelocity=500
     worldAlpha=1
     externalLinearDragV=<1 1 1>
    }
   }
   n1 = Control Rig(Source=Copy Pose From Mesh() {bUseAttachedParent=true}) {controlRigAssetReference {controlRigAsset=&/M_MED_MimicGlean_FaceAcc_CtrlRig_PreDyn}}
   n2 = "Constraint - dyn_hair_bk_01_base (PPP)"!(
    ComponentPose=Local To Component(LocalPose=@n1)
    Alpha=1
    ConstraintWeights_0=0.0279
    ConstraintWeights_1:float=0.6572
    ConstraintWeights_2:float=0.3149
   ) -> (Pose) {
    boneToModify {boneName=dyn_hair_bk_01_base}
    constraintSetup [{targetBone {boneName=neck_01}}, {targetBone {boneName=head}}, {targetBone {boneName=neck_02}}]
    constraintWeights [1,1,1]
   }
   n3 = "Constraint - dyn_hair_bk_02_base (PPP)"!(ComponentPose=@n2, Alpha=1, ConstraintWeights_0=0.0123, ConstraintWeights_1:float=0.6685, ConstraintWeights_2:float=0.3192) -> (Pose) {
    boneToModify {boneName=dyn_hair_bk_02_base}
    constraintSetup [{targetBone {boneName=neck_01}}, {targetBone {boneName=head}}, {targetBone {boneName=neck_02}}]
    constraintWeights [1,1,1]
   }
   n4 = "Constraint - dyn_hair_bk_03_base (PPP)"!(ComponentPose=@n3, Alpha=1, ConstraintWeights_0=0.012, ConstraintWeights_1:float=0.6725, ConstraintWeights_2:float=0.3155) -> (Pose) {
    boneToModify {boneName=dyn_hair_bk_03_base}
    constraintSetup [{targetBone {boneName=neck_01}}, {targetBone {boneName=head}}, {targetBone {boneName=neck_02}}]
    constraintWeights [1,1,1]
   }
   n5 = "Constraint - dyn_hair_bk_04_base (PPP)"!(ComponentPose=@n4, Alpha=1, ConstraintWeights_0=0.0238, ConstraintWeights_1:float=0.6525, ConstraintWeights_2:float=0.3237) -> (Pose) {
    boneToModify {boneName=dyn_hair_bk_04_base}
    constraintSetup [{targetBone {boneName=neck_01}}, {targetBone {boneName=head}}, {targetBone {boneName=neck_02}}]
    constraintWeights [1,1,1]
   }
   n6 = RigidBody(
    ComponentPose=@n5
    OverrideWorldGravity="0,0,0"
    ComponentLinearAccScale="0,0,0"
    ComponentLinearVelScale="0,0,0"
    bTransferBoneVelocities=False
   ) {
    alphaInputType=Bool
    bClampLinearTranslationLimitToRefPose=true
    bFreezeIncomingPoseOnStart=true
    bOverrideWorldGravity=true
    bTransferBoneVelocities=true
    overridePhysicsAsset=&M_MED_MimicGlean_FaceAcc_Physics
    simulationSpace=BaseBoneSpace
    alphaBoolBlend {blendOption=Linear}
    baseBoneRef {boneName=head}
    componentAppliedLinearAccClamp=<3000 3000 3500>
    componentLinearAccScale=<0.6 0.2 0.2>
    componentLinearVelScale=<0.6 0.2 0.2>
    simSpaceSettings {
     dampingAlpha=0
     maxAngularAcceleration=50
     maxAngularVelocity=8
     maxLinearAcceleration=1000
     maxLinearVelocity=500
     worldAlpha=1
     externalLinearDragV=<1 1 1>
    }
   }
   Output Pose(Result=Component To Local(ComponentPose=@n6)) {layerGroup=None}
  ]}
 EventGraph {nodes [
   Event BlueprintUpdateAnimation()
   > n1 = "Gravity Override State Function (Parameter Asset) (Nativized)"(
    CharacterPartAnimInstance=Self-Reference()
    GameTimeInSeconds=GetGameTimeInSeconds()
    DebugJointName=spine_05
    Parameters=/BRCosmetics/Characters/Player/Female/Medium/Bodies/F_MED_GoldenValley/Meshes/F_MED_GoldenValley_Parameters
   )
   > Set GravityOverride(GravityOverride=@n1)
  ]}
]
variables [GravityOverride {type=Vector}]
